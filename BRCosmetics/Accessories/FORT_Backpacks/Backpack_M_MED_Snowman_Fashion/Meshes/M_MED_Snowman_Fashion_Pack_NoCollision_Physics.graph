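# Physics asset for the Snowman Fashion backpack cosmetic, named as its no-collision variant, defining bodies and joints for its hair, gloves, handle, arm and head.
# Simple capsule and box bodies branch from the pack: arm, handle, head and a two-link glove chain, with two hair tufts hanging off the head.
# Most joints have tight angular limits, while the handle locks twist and the head-to-pack joint leaves all rotation free.
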PhysicsAsset /BRCosmetics/Accessories/FORT_Backpacks/Backpack_M_MED_Snowman_Fashion/Meshes/M_MED_Snowman_Fashion_Pack_NoCollision_Physics
bodies {
 arm1 [{shapeType=Capsule, length=3, radius=1, center=<3 0 0>, rotation=<rot 90 0 0>}]
 dyn_glove1 [{shapeType=Box, extentX=3.2528, extentY=7.6392, extentZ=7.6392, center=<5.4753 -0.4979 0.088>}]
 dyn_glove2 [{shapeType=Box, extentX=5.7867, extentY=7.4811, extentZ=8.4791, center=<4.6811 -0.1442 0.8192>}]
 dyn_hairLarge1 [{shapeType=Capsule, length=2.0895, radius=1.1157, center=<1.378 -0.0542 0.1169>, rotation=<rot 0 -85.0705 -84.9903>}]
 dyn_hairSmall1 [{shapeType=Capsule, length=1.0094, radius=0.6608, center=<0.913 0.0031 0.2397>, rotation=<rot 0 89.7388 -87.6623>}]
 dyn_handle [{shapeType=Box, extentX=5, extentY=10, extentZ=3, center=<2.5 0 0>}]
 head [{shapeType=Box, extentX=0.5, extentY=0.5, extentZ=0.5, center=<-0.1409 -0.2074 0.1526>}]
 pack [{shapeType=Capsule, length=1, radius=1}]
}
constraints [
 {bone1Name=dyn_hairSmall1, bone2Name=head, swing1LimitDegrees=5, swing2LimitDegrees=5, twistLimitDegrees=5}
 {bone1Name=dyn_hairLarge1, bone2Name=head, swing1LimitDegrees=5, swing2LimitDegrees=5, twistLimitDegrees=5}
 {bone1Name=dyn_glove1, bone2Name=pack, swing1LimitDegrees=5, swing2LimitDegrees=5, twistLimitDegrees=2}
 {bone1Name=dyn_glove2, bone2Name=dyn_glove1, swing1LimitDegrees=5, swing2LimitDegrees=5, twistLimitDegrees=2}
 {bone1Name=dyn_handle, bone2Name=pack, swing1LimitDegrees=5, swing2LimitDegrees=5, twistLimitDegrees=15, twistMotion=Locked}
 {
  bone1Name=head
  bone2Name=pack
  swing1LimitDegrees=45
  swing1Motion=Free
  swing2LimitDegrees=45
  swing2Motion=Free
  twistLimitDegrees=45
  twistMotion=Free
 }
 {bone1Name=arm1, bone2Name=pack, swing1LimitDegrees=5, swing2LimitDegrees=5, twistLimitDegrees=2}
]
