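# Physics asset for the Snowman Fashion backpack cosmetic, providing bodies and joints for its dangling scarf, tassels and zipper.
# The pack body, a sphere plus box, anchors front and back four-link scarf chains, two three-link tassel chains and a single zipper capsule.
# Scarf and zipper joints share moderate swing and twist limits, while tassels are stiffer and end in spheres.
# A tiny head box joins the pack with free swing and twist.
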PhysicsAsset /BRCosmetics/Accessories/FORT_Backpacks/Backpack_M_MED_Snowman_Fashion/Meshes/M_MED_Snowman_Fashion_Pack_Physics
bodies {
 dyn_scarfBack1 [{shapeType=Box, extentX=3.5, extentY=3, extentZ=1, center=<1.75 0 0>}]
 dyn_scarfBack2 [{shapeType=Box, extentX=3, extentY=3.5, extentZ=1, center=<1.5 0 0>}]
 dyn_scarfBack3 [{shapeType=Box, extentX=4, extentY=4.5, extentZ=1, center=<2 0 0>}]
 dyn_scarfBack4 [{shapeType=Box, extentX=6, extentY=6, extentZ=1, center=<3 0 0>}]
 dyn_scarfFront1 [{shapeType=Box, extentX=4, extentY=5, extentZ=1, center=<2 0 0>}]
 dyn_scarfFront2 [{shapeType=Box, extentX=3, extentY=5, extentZ=1, center=<1.5 0 0>}]
 dyn_scarfFront3 [{shapeType=Box, extentX=3.5, extentY=5.5233, extentZ=1, center=<1.75 0 0>}]
 dyn_scarfFront4 [{shapeType=Box, extentX=8, extentY=7, extentZ=1, center=<4 0 0>}]
 dyn_tassleBack1 [{shapeType=Box, extentX=2, extentY=1, extentZ=1, center=<1 0 0>}]
 dyn_tassleBack2 [{shapeType=Box, extentX=2, extentY=1, extentZ=1, center=<1 0 0>}]
 dyn_tassleBack3 [{shapeType=Sphere, radius=1.9032, center=<1.389 0.0014 0.2117>}]
 dyn_tassleFront1 [{shapeType=Box, extentX=2, extentY=1, extentZ=1, center=<1 0 0>}]
 dyn_tassleFront2 [{shapeType=Box, extentX=2, extentY=1, extentZ=1, center=<1 0 0>}]
 dyn_tassleFront3 [{shapeType=Sphere, radius=1.8798, center=<1.358 0.018 0.0533>}]
 dyn_zipper [{shapeType=Capsule, length=2, radius=0.75, center=<1.7327 0 0>, rotation=<rot -90 0 0>}]
 head [{shapeType=Box, extentX=0.5, extentY=0.5, extentZ=0.5, center=<-0.1409 -0.2074 0.1526>}]
 pack [
  {shapeType=Sphere, radius=9, center=<-0.3912 -15.623 -13.6945>}
  {shapeType=Box, extentX=16.6794, extentY=12.3942, extentZ=20.8583, center=<-2.4762 -5.3655 -5.7461>}
 ]
}
constraints [
 {bone1Name=dyn_tassleFront1, bone2Name=pack, swing1LimitDegrees=10, swing2LimitDegrees=10, twistLimitDegrees=3}
 {bone1Name=dyn_tassleFront2, bone2Name=dyn_tassleFront1, swing1LimitDegrees=10, swing2LimitDegrees=10, twistLimitDegrees=3}
 {bone1Name=dyn_tassleFront3, bone2Name=dyn_tassleFront2, swing1LimitDegrees=10, swing2LimitDegrees=10, twistLimitDegrees=3}
 {bone1Name=dyn_tassleBack1, bone2Name=pack, swing1LimitDegrees=10, swing2LimitDegrees=10, twistLimitDegrees=3}
 {bone1Name=dyn_tassleBack2, bone2Name=dyn_tassleBack1, swing1LimitDegrees=10, swing2LimitDegrees=10, twistLimitDegrees=3}
 {bone1Name=dyn_tassleBack3, bone2Name=dyn_tassleBack2, swing1LimitDegrees=10, swing2LimitDegrees=10, twistLimitDegrees=3}
 {bone1Name=dyn_zipper, bone2Name=pack, swing1LimitDegrees=25, swing2LimitDegrees=25, twistLimitDegrees=10}
 {bone1Name=dyn_scarfFront1, bone2Name=pack, swing1LimitDegrees=25, swing2LimitDegrees=25, twistLimitDegrees=10}
 {bone1Name=dyn_scarfFront2, bone2Name=dyn_scarfFront1, swing1LimitDegrees=25, swing2LimitDegrees=25, twistLimitDegrees=10}
 {bone1Name=dyn_scarfFront3, bone2Name=dyn_scarfFront2, swing1LimitDegrees=25, swing2LimitDegrees=25, twistLimitDegrees=10}
 {bone1Name=dyn_scarfFront4, bone2Name=dyn_scarfFront3, swing1LimitDegrees=25, swing2LimitDegrees=25, twistLimitDegrees=10}
 {bone1Name=dyn_scarfBack1, bone2Name=pack, swing1LimitDegrees=25, swing2LimitDegrees=25, twistLimitDegrees=10}
 {bone1Name=dyn_scarfBack2, bone2Name=dyn_scarfBack1, swing1LimitDegrees=25, swing2LimitDegrees=25, twistLimitDegrees=10}
 {bone1Name=dyn_scarfBack3, bone2Name=dyn_scarfBack2, swing1LimitDegrees=25, swing2LimitDegrees=25, twistLimitDegrees=10}
 {bone1Name=dyn_scarfBack4, bone2Name=dyn_scarfBack3, swing1LimitDegrees=25, swing2LimitDegrees=25, twistLimitDegrees=10}
 {
  bone1Name=head
  bone2Name=pack
  swing1LimitDegrees=45
  swing1Motion=Free
  swing2LimitDegrees=45
  swing2Motion=Free
  twistLimitDegrees=45
  twistMotion=Free
 }
]
